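# Physics asset for the Critter Cuddle backpack cosmetic, defining capsule bodies and joints for the critter's four legs, head and jaw.
# The legs and head each constrain to the pack body with tight angular limits, and the jaw chains off the head.
# The pack links to root with free swing and twist, while the narrow limits on the dyn_ bones allow only small secondary wobble.
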PhysicsAsset /BRCosmetics/Accessories/FORT_Backpacks/Backpack_M_MED_Critter_Cuddle/Meshes/M_MED_Critter_Cuddle_Pack_Physics
bodies {
 dyn_L_back_leg [{shapeType=Capsule, length=5, radius=1.5, shapeName=dyn_L_back_leg_capsule, center=<0.3862 1.3621 -3.7818>, rotation=<rot 0 -0.0001 -20.0004>}]
 dyn_L_front_leg [{
   shapeType=Capsule
   length=5
   radius=1.5
   shapeName=dyn_L_front_leg_capsule
   center=<-1.0508 -1.0383 -3.6544>
   rotation=<rot -18.7471 -7.0976 11.1719>
  }]
 dyn_R_back_leg [{
   shapeType=Capsule
   length=5.5
   radius=1.5
   shapeName=dyn_R_back_leg_capsule
   center=<1.4173 -1.4932 -4.1893>
   rotation=<rot 18.7471 7.0964 21.1723>
  }]
 dyn_R_front_leg [{
   shapeType=Capsule
   length=6.75
   radius=1.5
   shapeName=dyn_R_front_leg_capsule
   center=<-1.4043 1.7718 -3.9358>
   rotation=<rot -18.7471 7.0965 -21.1722>
  }]
 dyn_head_jnt [{shapeType=Capsule, length=7.25, radius=3.75, shapeName=dyn_head_jnt_capsule, center=<-5.285 0 -1.292>, rotation=<rot -60.0001 0 0>}]
 dyn_jaw_jnt [{shapeType=Capsule, length=7.5, radius=1.5, shapeName=dyn_jaw_jnt_capsule, center=<-4.6013 0 -1.9023>, rotation=<rot -60.0001 0 0>}]
 pack [{shapeType=Capsule, length=2, radius=2, shapeName=pack_capsule}]
 root [{shapeType=Capsule, length=2, radius=2, shapeName=root_capsule}]
}
constraints [
 {
  bone1Name=pack
  bone2Name=root
  swing1LimitDegrees=45
  swing1Motion=Free
  swing2LimitDegrees=45
  swing2Motion=Free
  twistLimitDegrees=45
  twistMotion=Free
 }
 {bone1Name=dyn_R_front_leg, bone2Name=pack, swing1LimitDegrees=15, swing2LimitDegrees=15, twistLimitDegrees=5}
 {bone1Name=dyn_L_front_leg, bone2Name=pack, swing1LimitDegrees=15, swing2LimitDegrees=15, twistLimitDegrees=5}
 {bone1Name=dyn_R_back_leg, bone2Name=pack, swing1LimitDegrees=15, swing2LimitDegrees=15, twistLimitDegrees=5}
 {bone1Name=dyn_L_back_leg, bone2Name=pack, swing1LimitDegrees=15, swing2LimitDegrees=15, twistLimitDegrees=5}
 {bone1Name=dyn_head_jnt, bone2Name=pack, swing1LimitDegrees=5, swing2LimitDegrees=25, twistLimitDegrees=10}
 {bone1Name=dyn_jaw_jnt, bone2Name=dyn_head_jnt, swing1LimitDegrees=5, swing2LimitDegrees=25, twistLimitDegrees=5}
]
